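# Spectator map player list row showing squad ID, name, health and shield bars, kills, event score and camera/followed icons.
# Player data, relevancy, death and disconnect events run GetPlayerListState into UpdateStyle, which sets the border style and shows hit points only for Default.
# GetPlayerListState sets Default or Irrelevant for non-eliminated players, never assigning Eliminated despite its red style.
# Events toggle squad ID and non-negative event score columns; PreConstruct previews DesignTime_State.
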WidgetBlueprint /FortniteCoreUI/Athena/HUD/Spectator/AthenaSpectatorMapPlayerListRow
graphs [
 EventGraph {nodes [
   n1 = Event PreConstruct()
   > Branch(Condition=@n1.IsDesignTime)
   > UpdateStyle!(
    execute
    self
    PlayerListState:ESpectatorMapPlayerListState Enum=Get DesignTime_State() -> (DesignTime_State:ESpectatorMapPlayerListState Enum=Default)
   ) -> (then)
   n2 = Event OnSquadIdVisibilityChanged!() -> (OutputDelegate, then, bVisible:bool) {is=title K2Node_Event}
   > SetVisibility(
    self=Get SizeBox_SquadId() -> (SizeBox_SquadId:Size Box Object Reference)
    InVisibility=Select(Option 0=Collapsed, Option 1=SelfHitTestInvisible, Index=@n2.bVisible)
   )
   n3 = Event OnIsInRelevancyChanged!() -> (OutputDelegate, then, bIsInRelevancy:bool) {is=title K2Node_Event}
   UpdateStyle(
    execute=[@n3.then, @n5.then, @n6.then, @n7.then]
    PlayerListState=GetPlayerListState!(self) -> (PlayerListState:ESpectatorMapPlayerListState Enum=Default)
   )
   n4 = Event OnEventScoreChanged!() -> (OutputDelegate, then, EventScore:int=0) {is=title K2Node_Event}
   > SetVisibility(
    self=Get SizeBox_EventScore() -> (SizeBox_EventScore:Size Box Object Reference)
    InVisibility=Select(
     Option 0=Collapsed
     Option 1=SelfHitTestInvisible
     Index="integer >= integer"!(A:int=@n4.EventScore, B:int=0) -> (ReturnValue:bool) {is=title K2Node_CallFunction}
    )
   )
   n5 = Event OnPlayerDataChanged!() -> (OutputDelegate, then, InPlayerData:Athena Remote Player View Data Object Reference) {is=title K2Node_Event}
   n6 = Event OnDeadStateChanged!() -> (OutputDelegate, then, bIsDead:bool) {is=title K2Node_Event}
   n7 = Event OnDisconnectedChanged!() -> (OutputDelegate, then, bIsDisconnected:bool) {is=title K2Node_Event}
  ]}
 GetPlayerListState {nodes [
   GetPlayerListState() {is=title K2Node_FunctionEntry}
   > n1 = Sequence()
   n2 = GetPlayerData!(self:Athena Spectator Player Widget Base Object Reference) -> (ReturnValue:Athena Remote Player View Data Object Reference)
   > Is Valid(InputObject=@n2)
   > n3 = Branch(Condition=GetIsEliminated!(self:Athena Remote Player View Data Object Reference=@n2) -> (ReturnValue:bool))
   Set Local_State(
    execute=@n3.else
    Local_State:ESpectatorMapPlayerListState Enum=Select(
     Option 0:ESpectatorMapPlayerListState Enum=Irrelevant
     Option 1:ESpectatorMapPlayerListState Enum=Default
     Index=GetIsInRelevancy!(self:Athena Remote Player View Data Object Reference=@n2) -> (ReturnValue:bool)
    ) -> (ReturnValue:ESpectatorMapPlayerListState Enum)
   ) -> (Output_Get:ESpectatorMapPlayerListState Enum=Default)
   Return Node!(
    execute=@n1.then_1
    PlayerListState:ESpectatorMapPlayerListState Enum=Get Local_State() -> (Local_State:ESpectatorMapPlayerListState Enum=Default)
   )
  ]}
 UpdateStyle {nodes [
   n1 = UpdateStyle!() -> (then, PlayerListState:ESpectatorMapPlayerListState Enum=Default) {is=title K2Node_FunctionEntry}
   > SetVisibility(
    self=Get HitPointInfoVB() -> (HitPointInfoVB:Vertical Box Object Reference)
    InVisibility=Select(
     Option 0:ESlateVisibility Enum=Collapsed
     Option 1:ESlateVisibility Enum=SelfHitTestInvisible
     Index="Equal (Enum)"(A:ESpectatorMapPlayerListState Enum=@n1, B:ESpectatorMapPlayerListState Enum=Default)
    ) -> (ReturnValue:ESlateVisibility Enum)
   )
   > SetStyle!(
    execute
    self:Common Border Object Reference=Get ScoreboardStateBorder() -> (ScoreboardStateBorder:Common Border Object Reference)
    InStyle:Common Border Style Class Reference=Select!(
     Default:Common Border Style Class Reference=Get DefaultStyle() -> (DefaultStyle:Common Border Style Class Reference)
     Irrelevant:Common Border Style Class Reference=Get IrrelevantStyle() -> (IrrelevantStyle:Common Border Style Class Reference)
     Eliminated:Common Border Style Class Reference=Get EliminatedStyle() -> (EliminatedStyle:Common Border Style Class Reference)
     Index:ESpectatorMapPlayerListState Enum=@n1
    ) -> (ReturnValue:Common Border Style Class Reference)
   ) -> (then)
  ]}
]
variables [
 DefaultStyle {type=Class, default=/FortUILibrary/Styles/Border/Border_AthenaStore_OfferHeaderBacking.Border_AthenaStore_OfferHeaderBacking_C}
 DesignTime_State {type=ESpectatorMapPlayerListState, default=Default}
 EliminatedStyle {type=Class, default=/FortUILibrary/Styles/Border/Border_StatRow_RedFill.Border_StatRow_RedFill_C}
 IrrelevantStyle {type=Class, default=/FortUILibrary/Styles/Border/Border-MainL-Black.Border-MainL-Black_C}
]
widgets {
 info {
  inheritedWidgetCount=10
  parentClass=/Script/FortniteUI.AthenaSpectatorPlayerListEntryBase
  rootWidgetClass=&Overlay
  widgetCount=22
 }
 widgets [
  Overlay_0 {widgetClassPath=&Overlay, widget=Overlay_0}
  ScoreboardStateBorder {
   widgetClassPath=&CommonBorder
   bIsVariable=true
   parent=Overlay_0
   widget=ScoreboardStateBorder
   properties {style=&Border_AthenaStore_OfferHeaderBacking_C, padding=<ltrb 0 0 0 0>}
   slot {class=OverlaySlot, properties {horizontalAlignment=HAlign_Fill, verticalAlignment=VAlign_Fill}}
  }
  HorizontalBox_0 {
   widgetClassPath=&HorizontalBox
   parent=ScoreboardStateBorder
   widget=HorizontalBox_0
   slot {class=BorderSlot, properties {padding=<ltrb 0 0 0 0>}}
  }
  SizeBox_SquadId {
   widgetClassPath=&SizeBox
   bIsVariable=true
   parent=HorizontalBox_0
   widget=SizeBox_SquadId
   properties {bOverride_WidthOverride=true, widthOverride=60}
   slot {class=HorizontalBoxSlot}
  }
  Border_SquadColor {
   widgetClassPath=&CommonBorder
   bInherited=true
   parent=SizeBox_SquadId
   widget=Border_SquadColor
   properties {
    horizontalAlignment=HAlign_Center
    verticalAlignment=VAlign_Center
    background {
     bIsDynamicallyLoaded=false
     drawAs=Box
     imageType=NoImage
     mirroring=NoMirror
     resourceName=None
     resourceObject=null
     tiling=NoTile
     imageSize=<32 32>
     margin=<ltrb 0 0 0 0>
     outlineSettings {
      bUseBrushTransparency=false
      roundingType=HalfHeightRadius
      width=0
      color {colorUseRule=UseColor_Specified, specifiedColor=<rgba 0 0 0 0>}
      cornerRadii=<0 0 0 0>
     }
     tintColor {colorUseRule=UseColor_Specified, specifiedColor=<rgba 1 1 1 1>}
     uVRegion {bIsValid=false, max=<0 0>, min=<0 0>}
    }
    brushColor=<rgba 1 0.25 0 1>
    padding=<ltrb 0 0 0 0>
   }
   slot {class=SizeBoxSlot}
  }
  Text_SquadId {
   widgetClassPath=&CommonTextBlock
   bInherited=true
   parent=Border_SquadColor
   widget=Text_SquadId
   properties {
    style=/FortUILibrary/Styles/Typography/Foundation/TextStyle-Base-S-B-Outline.TextStyle-Base-S-B-Outline_C
    text=888
    font {
     bForceMonospaced=false
     bMaterialIsStencil=false
     fontMaterial=null
     fontObject=&NotoSans
     letterSpacing=0
     monospacedWidth=1
     size=18
     skewAmount=0
     typefaceFontName=NotoSans-Bold
     outlineSettings {
      bApplyOutlineToDropShadows=false
      bMiteredCorners=false
      bSeparateFillAlpha=false
      outlineMaterial=null
      outlineSize=1
      outlineColor=<rgba 0 0 0 1>
     }
    }
    shadowOffset=<0 0>
   }
   slot {class=BorderSlot, properties {horizontalAlignment=HAlign_Center, verticalAlignment=VAlign_Center, padding=<ltrb 0 0 0 0>}}
  }
  SizeBox_2 {
   widgetClassPath=&SizeBox
   parent=HorizontalBox_0
   widget=SizeBox_2
   properties {bOverride_WidthOverride=true, widthOverride=300}
   slot {class=HorizontalBoxSlot}
  }
  VerticalBox_0 {
   widgetClassPath=&VerticalBox
   parent=SizeBox_2
   widget=VerticalBox_0
   slot {class=SizeBoxSlot, properties {verticalAlignment=VAlign_Center, padding=<ltrb 8 5 5 5>}}
  }
  Text_PlayerName {
   widgetClassPath=&CommonTextBlock
   bInherited=true
   parent=VerticalBox_0
   widget=Text_PlayerName
   properties {
    clipping=ClipToBounds
    scrollStyle=/FortUILibrary/Styles/Typography/Foundation/TextScrollStyle-NoFade.TextScrollStyle-NoFade_C
    style=/FortUILibrary/Styles/Typography/Foundation/TextStyle-Base-XXS-B.TextStyle-Base-XXS-B_C
    text=WWWWWWWWWWWWWWW
    font {
     bForceMonospaced=false
     bMaterialIsStencil=false
     fontMaterial=null
     fontObject=&NotoSans
     letterSpacing=0
     monospacedWidth=1
     size=14
     skewAmount=0
     typefaceFontName=NotoSans-Bold
     outlineSettings {
      bApplyOutlineToDropShadows=false
      bMiteredCorners=false
      bSeparateFillAlpha=false
      outlineMaterial=null
      outlineSize=0
      outlineColor=<rgba 0 0 0 1>
     }
    }
    shadowOffset=<0 0>
   }
   slot {class=VerticalBoxSlot, properties {verticalAlignment=VAlign_Center}}
  }
  HitPointInfoVB {widgetClassPath=&VerticalBox, bIsVariable=true, parent=VerticalBox_0, widget=HitPointInfoVB, slot {class=VerticalBoxSlot}}
  SizeBox_3 {
   widgetClassPath=&SizeBox
   parent=HitPointInfoVB
   widget=SizeBox_3
   properties {bOverride_HeightOverride=true, heightOverride=8}
   slot {class=VerticalBoxSlot, properties {padding=<ltrb 0 0 0 2>}}
  }
  HitPoint_ShieldBar {
   widgetClassPath=/Game/Athena/HUD/Spectator/AthenaSpectatorHitPointBar
   bInherited=true
   parent=SizeBox_3
   widget=HitPoint_ShieldBar
   properties {barType=Shield}
   slot {class=SizeBoxSlot}
  }
  HitPoint_HealthBar {
   widgetClassPath=&AthenaSpectatorHitPointBar
   bInherited=true
   parent=HitPointInfoVB
   widget=HitPoint_HealthBar
   slot {class=VerticalBoxSlot}
  }
  SizeBox_4 {
   widgetClassPath=&SizeBox
   parent=HorizontalBox_0
   widget=SizeBox_4
   properties {bOverride_WidthOverride=true, widthOverride=60}
   slot {class=HorizontalBoxSlot}
  }
  Text_KillScore {
   widgetClassPath=&CommonTextBlock
   bInherited=true
   parent=SizeBox_4
   widget=Text_KillScore
   properties {
    style=/FortUILibrary/Styles/Typography/Foundation/TextStyle-Base-S-B.TextStyle-Base-S-B_C
    text=888
    font {
     bForceMonospaced=false
     bMaterialIsStencil=false
     fontMaterial=null
     fontObject=&NotoSans
     letterSpacing=0
     monospacedWidth=1
     size=18
     skewAmount=0
     typefaceFontName=NotoSans-Bold
     outlineSettings {
      bApplyOutlineToDropShadows=false
      bMiteredCorners=false
      bSeparateFillAlpha=false
      outlineMaterial=null
      outlineSize=0
      outlineColor=<rgba 0 0 0 1>
     }
    }
    shadowOffset=<0 0>
   }
   slot {class=SizeBoxSlot, properties {horizontalAlignment=HAlign_Center, verticalAlignment=VAlign_Center}}
  }
  SizeBox_EventScore {
   widgetClassPath=&SizeBox
   bIsVariable=true
   parent=HorizontalBox_0
   widget=SizeBox_EventScore
   properties {bOverride_WidthOverride=true, widthOverride=60}
   slot {class=HorizontalBoxSlot}
  }
  Text_EventScore {
   widgetClassPath=&CommonTextBlock
   bInherited=true
   parent=SizeBox_EventScore
   widget=Text_EventScore
   properties {
    style=&TextStyle-Base-S-B_C
    text=888
    font {
     bForceMonospaced=false
     bMaterialIsStencil=false
     fontMaterial=null
     fontObject=&NotoSans
     letterSpacing=0
     monospacedWidth=1
     size=18
     skewAmount=0
     typefaceFontName=NotoSans-Bold
     outlineSettings {
      bApplyOutlineToDropShadows=false
      bMiteredCorners=false
      bSeparateFillAlpha=false
      outlineMaterial=null
      outlineSize=0
      outlineColor=<rgba 0 0 0 1>
     }
    }
    shadowOffset=<0 0>
   }
   slot {class=SizeBoxSlot, properties {horizontalAlignment=HAlign_Center, verticalAlignment=VAlign_Center}}
  }
  SizeBox_7 {
   widgetClassPath=&SizeBox
   parent=HorizontalBox_0
   widget=SizeBox_7
   properties {bOverride_WidthOverride=true, widthOverride=60}
   slot {class=HorizontalBoxSlot}
  }
  VerticalBox_1 {widgetClassPath=&VerticalBox, parent=SizeBox_7, widget=VerticalBox_1, slot {class=SizeBoxSlot}}
  Image_RecordingPlayerIcon {
   widgetClassPath=&Image
   bInherited=true
   parent=VerticalBox_1
   widget=Image_RecordingPlayerIcon
   properties {brush {
     bIsDynamicallyLoaded=false
     drawAs=Image
     imageType=NoImage
     mirroring=NoMirror
     resourceName=None
     resourceObject=/Game/UI/Foundation/Textures/Icons/Replay/T-Icon-Replay-CurrentCamera-L
     tiling=NoTile
     imageSize=<32 32>
     margin=<ltrb 0 0 0 0>
     outlineSettings {
      bUseBrushTransparency=false
      roundingType=HalfHeightRadius
      width=0
      color {colorUseRule=UseColor_Specified, specifiedColor=<rgba 0 0 0 0>}
      cornerRadii=<0 0 0 0>
     }
     tintColor {colorUseRule=UseColor_Specified, specifiedColor=<rgba 1 1 1 1>}
     uVRegion {bIsValid=false, max=<0 0>, min=<0 0>}
    }}
   slot {class=VerticalBoxSlot, properties {horizontalAlignment=HAlign_Center, verticalAlignment=VAlign_Center, size {sizeRule=Fill, value=1}}}
  }
  Image_FollowedPlayerIcon {
   widgetClassPath=&Image
   bInherited=true
   parent=VerticalBox_1
   widget=Image_FollowedPlayerIcon
   properties {brush {
     bIsDynamicallyLoaded=false
     drawAs=Image
     imageType=NoImage
     mirroring=NoMirror
     resourceName=None
     resourceObject=/Game/UI/Foundation/Textures/Icons/Replay/T-Icon-Replay-CurrentPlayer-L
     tiling=NoTile
     imageSize=<32 32>
     margin=<ltrb 0 0 0 0>
     outlineSettings {
      bUseBrushTransparency=false
      roundingType=HalfHeightRadius
      width=0
      color {colorUseRule=UseColor_Specified, specifiedColor=<rgba 0 0 0 0>}
      cornerRadii=<0 0 0 0>
     }
     tintColor {colorUseRule=UseColor_Specified, specifiedColor=<rgba 1 0.4969 0.013 1>}
     uVRegion {bIsValid=false, max=<0 0>, min=<0 0>}
    }}
   slot {class=VerticalBoxSlot, properties {horizontalAlignment=HAlign_Center, verticalAlignment=VAlign_Center, size {sizeRule=Fill, value=1}}}
  }
  Button {
   widgetClassPath=/Game/Athena/UI/Matchmaking/SimpleCommonButton
   bInherited=true
   parent=Overlay_0
   widget=Button
   properties {
    bApplyAlphaOnDisable=false
    bDisplayInputActionWhenNotInteractable=false
    bIsFocusable=true
    onButtonBaseClicked=(&self:WidgetTree.Button.HandleOnButtonClicked)
    onButtonBaseDoubleClicked=(&self:WidgetTree.Button.HandleOnButtonDoubleClicked)
    onButtonBaseHovered=(&self:WidgetTree.Button.HandleOnButtonHovered)
    onButtonBaseUnhovered=(&self:WidgetTree.Button.HandleOnButtonUnhovered)
    onSelectedChangedBase=(&self:WidgetTree.Button.HandleOnSelectedChanged)
    style=/FortUILibrary/Foundation/Buttons/ButtonStyles/Athena/ButtonStyle_TransparentList.ButtonStyle_TransparentList_C
   }
   slot {class=OverlaySlot, properties {horizontalAlignment=HAlign_Fill, verticalAlignment=VAlign_Fill}}
  }
 ]
}
