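# Spectator map player list widget: sortable column headers for squad ID, player name, eliminations, event score and state above a list of player rows.
# Built on AthenaSpectatorMapPlayerListBase, which supplies the sort buttons and ListView; the graph only handles presentation events.
# On sort method change, a SortMethodToArrow macro picks that column's arrow to tint active, dimming the rest.
# Squad ID column visibility follows its event; event score column appears only for scores above zero.
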WidgetBlueprint /FortniteCoreUI/HUD/Spectator/AthenaSpectatorMapPlayerList
graphs [
 EventGraph {nodes [
   n1 = Event Handle Current Sort Method Changed!() -> (OutputDelegate, then, NewSortMethod:ESpectatorPlayerListSortMethod Enum=SquadId) {is=title K2Node_Event}
   > n2 = For Each Loop(Array:Array of Image Object References=Get SortArrows() -> (SortArrows:Array of Image Object References)) -> (Array Element:Image Object Reference)
   n3 = Sort Method to Arrow!(SortMethod:ESpectatorPlayerListSortMethod Enum=@n1.NewSortMethod) -> (Arrow:Image Object Reference) {is=title K2Node_MacroInstance}
   > SetColorAndOpacity(
    self:Image Object Reference=@n2
    InColorAndOpacity:Linear Color Structure=SelectColor(
     A=Get SortArrowActiveColor() -> (SortArrowActiveColor:Linear Color Structure)
     B=Get SortArrowInactiveColor() -> (SortArrowInactiveColor:Linear Color Structure)
     bPickA="Equal (Object)"(A=@n2, B=@n3) {is=title K2Node_CallFunction}
    )
   )
   n4 = Event Handle Player Event Score Changed!() -> (OutputDelegate, then, EventScore:int=0) {is=title K2Node_Event}
   > SetVisibility(
    self=Get SizeBox_EventScore() -> (SizeBox_EventScore:Size Box Object Reference)
    InVisibility=Select(Option 0=Collapsed, Option 1=SelfHitTestInvisible, Index="integer > integer"(A=@n4.EventScore, B=0) {is=title K2Node_CallFunction})
   )
   Event OnInitialized()
   > Set SortArrows(SortArrows:Array of Image Object References=Make Array!(
     "[0]":Image Object Reference=Get SortBySquadIdDescending() -> (SortBySquadIdDescending:Image Object Reference)
     "[1]":Image Object Reference=Get SortByPlayerNameDescending() -> (SortByPlayerNameDescending:Image Object Reference)
     "[2]":Image Object Reference=Get SortByEliminationsDescending() -> (SortByEliminationsDescending:Image Object Reference)
     "[3]":Image Object Reference=Get SortByEventScoreDescending() -> (SortByEventScoreDescending:Image Object Reference)
     "[4]":Image Object Reference=Get SortByStateDescending() -> (SortByStateDescending:Image Object Reference)
    ) -> (Array:Array of Image Object References)) -> (Output_Get:Array of Image Object References)
   n5 = Event Handle Squad Id Visibility Changed!() -> (OutputDelegate, then, bSquadIdsVisible:bool) {is=title K2Node_Event}
   > SetVisibility(
    self=Get SquadIdBox() -> (SquadIdBox:Size Box Object Reference)
    InVisibility=Select(Option 0=Collapsed, Option 1=SelfHitTestInvisible, Index=@n5.bSquadIdsVisible)
   )
  ]}
 SortMethodToArrow {nodes [Outputs!(Arrow:Image Object Reference=Select!(
     SquadId:Image Object Reference=Get SortBySquadIdDescending()
     PlayerName:Image Object Reference=Get SortByPlayerNameDescending()
     Eliminations:Image Object Reference=Get SortByEliminationsDescending()
     EventScore:Image Object Reference=Get SortByEventScoreDescending()
     State:Image Object Reference=Get SortByStateDescending()
     Index:ESpectatorPlayerListSortMethod Enum=Inputs!() -> (SortMethod:ESpectatorPlayerListSortMethod Enum=SquadId)
    ) -> (ReturnValue:Image Object Reference))]}
]
variables [
 SortArrowActiveColor {type=LinearColor, default=<rgba 1 1 1 1>}
 SortArrowInactiveColor {type=LinearColor, default=<rgba 0.0037 0.0075 0.0232 1>}
 SortArrows {type="array<Image>", default []}
]
widgets {
 info {
  inheritedWidgetCount=6
  parentClass=/Script/FortniteUI.AthenaSpectatorMapPlayerListBase
  rootWidgetClass=&CommonBorder
  widgetCount=39
 }
 widgets [
  CommonBorder_0 {
   widgetClassPath=&CommonBorder
   widget=CommonBorder_0
   properties {style=/FortUILibrary/Styles/Border/Border_Navy_VGrad.Border_Navy_VGrad_C, padding=<ltrb 0 0 0 0>}
  }
  VerticalBox_0 {
   widgetClassPath=&VerticalBox
   parent=CommonBorder_0
   widget=VerticalBox_0
   slot {class=BorderSlot, properties {padding=<ltrb 0 0 0 0>}}
  }
  HorizontalBox_0 {
   widgetClassPath=&HorizontalBox
   parent=VerticalBox_0
   widget=HorizontalBox_0
   slot {class=VerticalBoxSlot, properties {padding=<ltrb 5 0 5 0>}}
  }
  SquadIdBox {
   widgetClassPath=&SizeBox
   bIsVariable=true
   parent=HorizontalBox_0
   widget=SquadIdBox
   properties {bOverride_WidthOverride=true, widthOverride=60}
   slot {class=HorizontalBoxSlot}
  }
  Overlay_2 {
   widgetClassPath=&Overlay
   parent=SquadIdBox
   widget=Overlay_2
   slot {class=SizeBoxSlot, properties {horizontalAlignment=HAlign_Center}}
  }
  Button_SortBySquadId {
   widgetClassPath=&IconTextButton
   bInherited=true
   parent=Overlay_2
   widget=Button_SortBySquadId
   properties {
    bHideInputAction=true
    bIsFocusable=false
    buttonText=""
    onButtonBaseClicked=(&self:WidgetTree.Button_SortBySquadId.HandleOnButtonClicked)
    onButtonBaseDoubleClicked=(&self:WidgetTree.Button_SortBySquadId.HandleOnButtonDoubleClicked)
    onButtonBaseHovered=(&self:WidgetTree.Button_SortBySquadId.HandleOnButtonHovered)
    onButtonBaseUnhovered=(&self:WidgetTree.Button_SortBySquadId.HandleOnButtonUnhovered)
    onSelectedChangedBase=(&self:WidgetTree.Button_SortBySquadId.HandleOnSelectedChanged)
    style=/FortUILibrary/Foundation/Buttons/ButtonStyles/ButtonStyle-TransparentWithCues-XS-B-White.ButtonStyle-TransparentWithCues-XS-B-White_C
   }
   slot {class=OverlaySlot, properties {horizontalAlignment=HAlign_Fill, verticalAlignment=VAlign_Fill}}
  }
  HorizontalBox_1 {
   widgetClassPath=&HorizontalBox
   parent=Overlay_2
   widget=HorizontalBox_1
   slot {class=OverlaySlot, properties {horizontalAlignment=HAlign_Center, verticalAlignment=VAlign_Fill}}
  }
  CommonTextBlock_0 {
   widgetClassPath=&CommonTextBlock
   parent=HorizontalBox_1
   widget=CommonTextBlock_0
   properties {
    justification=Center
    style=/FortUILibrary/Styles/Typography/Foundation/TextStyle-Base-M-B.TextStyle-Base-M-B_C
    text="##"
    font {
     bForceMonospaced=false
     bMaterialIsStencil=false
     fontMaterial=null
     fontObject=&NotoSans
     letterSpacing=0
     monospacedWidth=1
     size=24
     skewAmount=0
     typefaceFontName=NotoSans-Bold
     outlineSettings {
      bApplyOutlineToDropShadows=false
      bMiteredCorners=false
      bSeparateFillAlpha=false
      outlineMaterial=null
      outlineSize=0
      outlineColor=<rgba 0 0 0 1>
     }
    }
    shadowOffset=<0 0>
   }
   slot {class=HorizontalBoxSlot, properties {verticalAlignment=VAlign_Center, padding=<ltrb 0 0 2 0>}}
  }
  VerticalBox_4 {
   widgetClassPath=&VerticalBox
   parent=HorizontalBox_1
   widget=VerticalBox_4
   slot {class=HorizontalBoxSlot, properties {verticalAlignment=VAlign_Center}}
  }
  SortBySquadIdDescending {
   widgetClassPath=&Image
   bIsVariable=true
   parent=VerticalBox_4
   widget=SortBySquadIdDescending
   properties {
    visibility=SelfHitTestInvisible
    brush {
     bIsDynamicallyLoaded=false
     drawAs=Image
     imageType=NoImage
     mirroring=NoMirror
     resourceName=None
     resourceObject=/Game/Widgets/SActorCanvas/DirectionArrow
     tiling=NoTile
     imageSize=<14 14>
     margin=<ltrb 0 0 0 0>
     outlineSettings {
      bUseBrushTransparency=false
      roundingType=HalfHeightRadius
      width=0
      color {colorUseRule=UseColor_Specified, specifiedColor=<rgba 0 0 0 0>}
      cornerRadii=<0 0 0 0>
     }
     tintColor {colorUseRule=UseColor_Specified, specifiedColor=<rgba 1 1 1 1>}
     uVRegion {bIsValid=false, max=<0 0>, min=<0 0>}
    }
    renderTransform {angle=180, scale=<1 1>, shear=<0 0>, translation=<0 0>}
   }
   slot {class=VerticalBoxSlot, properties {horizontalAlignment=HAlign_Center, verticalAlignment=VAlign_Center, padding=<ltrb 0 -1 0 0>}}
  }
  SizeBox_PlayerName {
   widgetClassPath=&SizeBox
   parent=HorizontalBox_0
   widget=SizeBox_PlayerName
   properties {bOverride_WidthOverride=true, widthOverride=300}
   slot {class=HorizontalBoxSlot, properties {horizontalAlignment=HAlign_Center}}
  }
  Overlay_3 {
   widgetClassPath=&Overlay
   parent=SizeBox_PlayerName
   widget=Overlay_3
   slot {class=SizeBoxSlot, properties {horizontalAlignment=HAlign_Center}}
  }
  Button_SortByPlayerName {
   widgetClassPath=&IconTextButton
   bInherited=true
   parent=Overlay_3
   widget=Button_SortByPlayerName
   properties {
    bHideInputAction=true
    bIsFocusable=false
    buttonText=""
    onButtonBaseClicked=(&self:WidgetTree.Button_SortByPlayerName.HandleOnButtonClicked)
    onButtonBaseDoubleClicked=(&self:WidgetTree.Button_SortByPlayerName.HandleOnButtonDoubleClicked)
    onButtonBaseHovered=(&self:WidgetTree.Button_SortByPlayerName.HandleOnButtonHovered)
    onButtonBaseUnhovered=(&self:WidgetTree.Button_SortByPlayerName.HandleOnButtonUnhovered)
    onSelectedChangedBase=(&self:WidgetTree.Button_SortByPlayerName.HandleOnSelectedChanged)
    style=&ButtonStyle-TransparentWithCues-XS-B-White_C
   }
   slot {class=OverlaySlot, properties {horizontalAlignment=HAlign_Fill, verticalAlignment=VAlign_Fill}}
  }
  HorizontalBox_2 {
   widgetClassPath=&HorizontalBox
   parent=Overlay_3
   widget=HorizontalBox_2
   slot {class=OverlaySlot, properties {horizontalAlignment=HAlign_Center, verticalAlignment=VAlign_Fill}}
  }
  Image_7 {
   widgetClassPath=&Image
   parent=HorizontalBox_2
   widget=Image_7
   properties {
    visibility=SelfHitTestInvisible
    brush {
     bIsDynamicallyLoaded=false
     drawAs=Image
     imageType=NoImage
     mirroring=NoMirror
     resourceName=None
     resourceObject=/Game/UI/Foundation/Textures/Icons/Replay/T-Icon-Replay-CurrentPlayer-L
     tiling=NoTile
     imageSize=<36 36>
     margin=<ltrb 0 0 0 0>
     outlineSettings {
      bUseBrushTransparency=false
      roundingType=HalfHeightRadius
      width=0
      color {colorUseRule=UseColor_Specified, specifiedColor=<rgba 0 0 0 0>}
      cornerRadii=<0 0 0 0>
     }
     tintColor {colorUseRule=UseColor_Specified, specifiedColor=<rgba 1 1 1 1>}
     uVRegion {bIsValid=false, max=<0 0>, min=<0 0>}
    }
   }
   slot {class=HorizontalBoxSlot, properties {verticalAlignment=VAlign_Center, padding=<ltrb 0 0 2 0>}}
  }
  VerticalBox_5 {
   widgetClassPath=&VerticalBox
   parent=HorizontalBox_2
   widget=VerticalBox_5
   slot {class=HorizontalBoxSlot, properties {verticalAlignment=VAlign_Center}}
  }
  SortByPlayerNameDescending {
   widgetClassPath=&Image
   bIsVariable=true
   parent=VerticalBox_5
   widget=SortByPlayerNameDescending
   properties {
    visibility=SelfHitTestInvisible
    brush {
     bIsDynamicallyLoaded=false
     drawAs=Image
     imageType=NoImage
     mirroring=NoMirror
     resourceName=None
     resourceObject=&DirectionArrow
     tiling=NoTile
     imageSize=<14 14>
     margin=<ltrb 0 0 0 0>
     outlineSettings {
      bUseBrushTransparency=false
      roundingType=HalfHeightRadius
      width=0
      color {colorUseRule=UseColor_Specified, specifiedColor=<rgba 0 0 0 0>}
      cornerRadii=<0 0 0 0>
     }
     tintColor {colorUseRule=UseColor_Specified, specifiedColor=<rgba 1 1 1 1>}
     uVRegion {bIsValid=false, max=<0 0>, min=<0 0>}
    }
    renderTransform {angle=180, scale=<1 1>, shear=<0 0>, translation=<0 0>}
   }
   slot {class=VerticalBoxSlot, properties {horizontalAlignment=HAlign_Center, verticalAlignment=VAlign_Center, padding=<ltrb 0 -1 0 0>}}
  }
  SizeBox_Eliminations {
   widgetClassPath=&SizeBox
   parent=HorizontalBox_0
   widget=SizeBox_Eliminations
   properties {bOverride_WidthOverride=true, widthOverride=60}
   slot {class=HorizontalBoxSlot}
  }
  Overlay_4 {
   widgetClassPath=&Overlay
   parent=SizeBox_Eliminations
   widget=Overlay_4
   slot {class=SizeBoxSlot, properties {horizontalAlignment=HAlign_Center}}
  }
  Button_SortByEliminations {
   widgetClassPath=&IconTextButton
   bInherited=true
   parent=Overlay_4
   widget=Button_SortByEliminations
   properties {
    bIsFocusable=false
    buttonText=""
    onButtonBaseClicked=(&self:WidgetTree.Button_SortByEliminations.HandleOnButtonClicked)
    onButtonBaseDoubleClicked=(&self:WidgetTree.Button_SortByEliminations.HandleOnButtonDoubleClicked)
    onButtonBaseHovered=(&self:WidgetTree.Button_SortByEliminations.HandleOnButtonHovered)
    onButtonBaseUnhovered=(&self:WidgetTree.Button_SortByEliminations.HandleOnButtonUnhovered)
    onSelectedChangedBase=(&self:WidgetTree.Button_SortByEliminations.HandleOnSelectedChanged)
    style=&ButtonStyle-TransparentWithCues-XS-B-White_C
   }
   slot {class=OverlaySlot, properties {horizontalAlignment=HAlign_Fill, verticalAlignment=VAlign_Fill}}
  }
  HorizontalBox_3 {
   widgetClassPath=&HorizontalBox
   parent=Overlay_4
   widget=HorizontalBox_3
   slot {class=OverlaySlot, properties {horizontalAlignment=HAlign_Center, verticalAlignment=VAlign_Fill}}
  }
  KillsIcon {
   widgetClassPath=&Image
   parent=HorizontalBox_3
   widget=KillsIcon
   properties {
    visibility=SelfHitTestInvisible
    brush {
     bIsDynamicallyLoaded=false
     drawAs=Image
     imageType=NoImage
     mirroring=NoMirror
     resourceName=None
     resourceObject=/Game/UI/Foundation/Textures/Icons/Replay/T-Icon-Replay-Filter-Kill-L
     tiling=NoTile
     imageSize=<36 36>
     margin=<ltrb 0 0 0 0>
     outlineSettings {
      bUseBrushTransparency=false
      roundingType=HalfHeightRadius
      width=0
      color {colorUseRule=UseColor_Specified, specifiedColor=<rgba 0 0 0 0>}
      cornerRadii=<0 0 0 0>
     }
     tintColor {colorUseRule=UseColor_Specified, specifiedColor=<rgba 1 1 1 1>}
     uVRegion {bIsValid=false, max=<0 0>, min=<0 0>}
    }
   }
   slot {class=HorizontalBoxSlot, properties {verticalAlignment=VAlign_Center, padding=<ltrb 0 0 2 0>}}
  }
  VerticalBox_6 {
   widgetClassPath=&VerticalBox
   parent=HorizontalBox_3
   widget=VerticalBox_6
   slot {class=HorizontalBoxSlot, properties {verticalAlignment=VAlign_Center}}
  }
  SortByEliminationsDescending {
   widgetClassPath=&Image
   bIsVariable=true
   parent=VerticalBox_6
   widget=SortByEliminationsDescending
   properties {
    visibility=SelfHitTestInvisible
    brush {
     bIsDynamicallyLoaded=false
     drawAs=Image
     imageType=NoImage
     mirroring=NoMirror
     resourceName=None
     resourceObject=&DirectionArrow
     tiling=NoTile
     imageSize=<14 14>
     margin=<ltrb 0 0 0 0>
     outlineSettings {
      bUseBrushTransparency=false
      roundingType=HalfHeightRadius
      width=0
      color {colorUseRule=UseColor_Specified, specifiedColor=<rgba 0 0 0 0>}
      cornerRadii=<0 0 0 0>
     }
     tintColor {colorUseRule=UseColor_Specified, specifiedColor=<rgba 1 1 1 1>}
     uVRegion {bIsValid=false, max=<0 0>, min=<0 0>}
    }
    renderTransform {angle=180, scale=<1 1>, shear=<0 0>, translation=<0 0>}
   }
   slot {class=VerticalBoxSlot, properties {horizontalAlignment=HAlign_Center, verticalAlignment=VAlign_Center, padding=<ltrb 0 -1 0 0>}}
  }
  SizeBox_EventScore {
   widgetClassPath=&SizeBox
   bIsVariable=true
   parent=HorizontalBox_0
   widget=SizeBox_EventScore
   properties {bOverride_WidthOverride=true, visibility=Collapsed, widthOverride=60}
   slot {class=HorizontalBoxSlot}
  }
  Overlay {
   widgetClassPath=&Overlay
   parent=SizeBox_EventScore
   widget=Overlay
   slot {class=SizeBoxSlot, properties {horizontalAlignment=HAlign_Center}}
  }
  Button_SortByEventScore {
   widgetClassPath=&IconTextButton
   bInherited=true
   parent=Overlay
   widget=Button_SortByEventScore
   properties {
    bIsFocusable=false
    buttonText=""
    onButtonBaseClicked=(&self:WidgetTree.Button_SortByEventScore.HandleOnButtonClicked)
    onButtonBaseDoubleClicked=(&self:WidgetTree.Button_SortByEventScore.HandleOnButtonDoubleClicked)
    onButtonBaseHovered=(&self:WidgetTree.Button_SortByEventScore.HandleOnButtonHovered)
    onButtonBaseUnhovered=(&self:WidgetTree.Button_SortByEventScore.HandleOnButtonUnhovered)
    onSelectedChangedBase=(&self:WidgetTree.Button_SortByEventScore.HandleOnSelectedChanged)
    style=&ButtonStyle-TransparentWithCues-XS-B-White_C
   }
   slot {class=OverlaySlot, properties {horizontalAlignment=HAlign_Fill, verticalAlignment=VAlign_Fill}}
  }
  HorizontalBox {
   widgetClassPath=&HorizontalBox
   parent=Overlay
   widget=HorizontalBox
   slot {class=OverlaySlot, properties {horizontalAlignment=HAlign_Center, verticalAlignment=VAlign_Fill}}
  }
  EventScoreIcon {
   widgetClassPath=&Image
   parent=HorizontalBox
   widget=EventScoreIcon
   properties {
    visibility=SelfHitTestInvisible
    brush {
     bIsDynamicallyLoaded=false
     drawAs=Image
     imageType=NoImage
     mirroring=NoMirror
     resourceName=None
     resourceObject=/FortniteCoreUI/Frontend/Showdown/Materials/T_UI_Ingame_TournamentScorePlate_InnerPlate
     tiling=NoTile
     imageSize=<36 36>
     margin=<ltrb 0 0 0 0>
     outlineSettings {
      bUseBrushTransparency=false
      roundingType=HalfHeightRadius
      width=0
      color {colorUseRule=UseColor_Specified, specifiedColor=<rgba 0 0 0 0>}
      cornerRadii=<0 0 0 0>
     }
     tintColor {colorUseRule=UseColor_Specified, specifiedColor=<rgba 1 1 1 1>}
     uVRegion {bIsValid=false, max=<0 0>, min=<0 0>}
    }
   }
   slot {class=HorizontalBoxSlot, properties {verticalAlignment=VAlign_Center, padding=<ltrb 0 0 2 0>}}
  }
  VerticalBox {
   widgetClassPath=&VerticalBox
   parent=HorizontalBox
   widget=VerticalBox
   slot {class=HorizontalBoxSlot, properties {verticalAlignment=VAlign_Center}}
  }
  SortByEventScoreDescending {
   widgetClassPath=&Image
   bIsVariable=true
   parent=VerticalBox
   widget=SortByEventScoreDescending
   properties {
    visibility=SelfHitTestInvisible
    brush {
     bIsDynamicallyLoaded=false
     drawAs=Image
     imageType=NoImage
     mirroring=NoMirror
     resourceName=None
     resourceObject=&DirectionArrow
     tiling=NoTile
     imageSize=<14 14>
     margin=<ltrb 0 0 0 0>
     outlineSettings {
      bUseBrushTransparency=false
      roundingType=HalfHeightRadius
      width=0
      color {colorUseRule=UseColor_Specified, specifiedColor=<rgba 0 0 0 0>}
      cornerRadii=<0 0 0 0>
     }
     tintColor {colorUseRule=UseColor_Specified, specifiedColor=<rgba 1 1 1 1>}
     uVRegion {bIsValid=false, max=<0 0>, min=<0 0>}
    }
    renderTransform {angle=180, scale=<1 1>, shear=<0 0>, translation=<0 0>}
   }
   slot {class=VerticalBoxSlot, properties {horizontalAlignment=HAlign_Center, verticalAlignment=VAlign_Center, padding=<ltrb 0 -1 0 0>}}
  }
  SizeBox_State {
   widgetClassPath=&SizeBox
   parent=HorizontalBox_0
   widget=SizeBox_State
   properties {bOverride_WidthOverride=true, widthOverride=60}
   slot {class=HorizontalBoxSlot}
  }
  Overlay_0 {
   widgetClassPath=&Overlay
   parent=SizeBox_State
   widget=Overlay_0
   slot {class=SizeBoxSlot, properties {horizontalAlignment=HAlign_Center}}
  }
  Button_SortByState {
   widgetClassPath=&IconTextButton
   bInherited=true
   parent=Overlay_0
   widget=Button_SortByState
   properties {
    bIsFocusable=false
    buttonText=""
    onButtonBaseClicked=(&self:WidgetTree.Button_SortByState.HandleOnButtonClicked)
    onButtonBaseDoubleClicked=(&self:WidgetTree.Button_SortByState.HandleOnButtonDoubleClicked)
    onButtonBaseHovered=(&self:WidgetTree.Button_SortByState.HandleOnButtonHovered)
    onButtonBaseUnhovered=(&self:WidgetTree.Button_SortByState.HandleOnButtonUnhovered)
    onSelectedChangedBase=(&self:WidgetTree.Button_SortByState.HandleOnSelectedChanged)
    style=&ButtonStyle-TransparentWithCues-XS-B-White_C
   }
   slot {class=OverlaySlot, properties {horizontalAlignment=HAlign_Fill, verticalAlignment=VAlign_Fill}}
  }
  HorizontalBox_5 {
   widgetClassPath=&HorizontalBox
   parent=Overlay_0
   widget=HorizontalBox_5
   slot {class=OverlaySlot, properties {horizontalAlignment=HAlign_Center, verticalAlignment=VAlign_Fill}}
  }
  Image_0 {
   widgetClassPath=&Image
   parent=HorizontalBox_5
   widget=Image_0
   properties {
    visibility=SelfHitTestInvisible
    brush {
     bIsDynamicallyLoaded=false
     drawAs=Image
     imageType=NoImage
     mirroring=NoMirror
     resourceName=None
     resourceObject=/Game/UI/Foundation/Textures/Icons/Replay/T-Icon-Replay-Exclamation-L
     tiling=NoTile
     imageSize=<36 36>
     margin=<ltrb 0 0 0 0>
     outlineSettings {
      bUseBrushTransparency=false
      roundingType=HalfHeightRadius
      width=0
      color {colorUseRule=UseColor_Specified, specifiedColor=<rgba 0 0 0 0>}
      cornerRadii=<0 0 0 0>
     }
     tintColor {colorUseRule=UseColor_Specified, specifiedColor=<rgba 1 1 1 1>}
     uVRegion {bIsValid=false, max=<0 0>, min=<0 0>}
    }
   }
   slot {class=HorizontalBoxSlot, properties {verticalAlignment=VAlign_Center, padding=<ltrb 0 0 2 0>}}
  }
  VerticalBox_1 {
   widgetClassPath=&VerticalBox
   parent=HorizontalBox_5
   widget=VerticalBox_1
   slot {class=HorizontalBoxSlot, properties {verticalAlignment=VAlign_Center}}
  }
  SortByStateDescending {
   widgetClassPath=&Image
   bIsVariable=true
   parent=VerticalBox_1
   widget=SortByStateDescending
   properties {
    visibility=SelfHitTestInvisible
    brush {
     bIsDynamicallyLoaded=false
     drawAs=Image
     imageType=NoImage
     mirroring=NoMirror
     resourceName=None
     resourceObject=&DirectionArrow
     tiling=NoTile
     imageSize=<14 14>
     margin=<ltrb 0 0 0 0>
     outlineSettings {
      bUseBrushTransparency=false
      roundingType=HalfHeightRadius
      width=0
      color {colorUseRule=UseColor_Specified, specifiedColor=<rgba 0 0 0 0>}
      cornerRadii=<0 0 0 0>
     }
     tintColor {colorUseRule=UseColor_Specified, specifiedColor=<rgba 1 1 1 1>}
     uVRegion {bIsValid=false, max=<0 0>, min=<0 0>}
    }
    renderTransform {angle=180, scale=<1 1>, shear=<0 0>, translation=<0 0>}
   }
   slot {class=VerticalBoxSlot, properties {horizontalAlignment=HAlign_Center, verticalAlignment=VAlign_Center, padding=<ltrb 0 -1 0 0>}}
  }
  ListView {
   widgetClassPath=&CommonListView
   bInherited=true
   parent=VerticalBox_0
   widget=ListView
   properties {
    entryWidgetClass=/FortniteCoreUI/Athena/HUD/Spectator/AthenaSpectatorMapPlayerListRow.AthenaSpectatorMapPlayerListRow_C
    horizontalEntrySpacing=2
    navigation=&self:WidgetTree.ListView.WidgetNavigation_0
    verticalEntrySpacing=2
   }
   slot {class=VerticalBoxSlot, properties {padding=<ltrb 5 0 5 5>, size {sizeRule=Fill, value=1}}}
  }
 ]
}
